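# Physics asset for the Love Queen backpack cosmetic, defining collision and joint limits for a dangling chain.
# Six chain link bodies are constrained in series from dyn_chain_mstr, with tight twist and wider swing limits.
# The end link combines three capsules at different orientations into a bulkier tip shape.
# Large pelvis and spine capsules have no constraints and stand in for the wearer's torso as collision volumes.
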PhysicsAsset /BRCosmetics/Accessories/FORT_Backpacks/Backpack_F_MED_Love_Queen/Meshes/F_MED_Love_Queen_Pack_Physics
bodies {
 dyn_chain_01 [{shapeType=Capsule, length=0.7403, radius=0.5, shapeName=dyn_chain_01_capsule, center=<0.3978 0 0>, rotation=<rot -90 0 0>}]
 dyn_chain_02 [{shapeType=Capsule, length=0.6604, radius=0.5799, shapeName=dyn_chain_02_capsule, center=<0.3978 0 0>, rotation=<rot -90 -0.0099 0.0099>}]
 dyn_chain_03 [{shapeType=Capsule, length=0.6093, radius=0.6311, shapeName=dyn_chain_03_capsule, center=<0.3978 0 0>, rotation=<rot -90 -0.0099 0.0099>}]
 dyn_chain_04 [{shapeType=Capsule, length=1.0208, radius=0.6594, shapeName=dyn_chain_04_capsule, center=<0.3978 0 0>, rotation=<rot -90 -0.0099 0.0099>}]
 dyn_chain_05 [{shapeType=Capsule, length=1.0208, radius=0.6594, shapeName=dyn_chain_05_capsule, center=<0.3978 0 0>, rotation=<rot -90 -0.0099 0.0099>}]
 dyn_chain_06 [
  {shapeType=Capsule, length=2.5924, radius=1.2486, shapeName=dyn_chain_06_capsule, center=<2.4928 0 0>, rotation=<rot 90 0 0>}
  {shapeType=Capsule, length=2.8409, radius=0.7246, shapeName=dyn_chain_06_capsule, center=<1.5091 0 0>, rotation=<rot -90 -5.7106 95.7077>}
  {
   shapeType=Capsule
   length=2.8409
   radius=0.7246
   shapeName=dyn_chain_06_capsule
   center=<1.5108 0.0005 0>
   rotation=<rot -0.0007 -90.0028 179.9999>
  }
 ]
 dyn_chain_mstr [{shapeType=Capsule, length=0.1, radius=0.1, shapeName=dyn_chain_mstr_capsule, center=<-0.4288 0 0>}]
 pelvis [{shapeType=Capsule, length=5, radius=11.25, shapeName=pelvis_capsule, center=<0 -2.2007 0>}]
 spine_01 [{shapeType=Capsule, length=5.5, radius=10.5, shapeName=spine_01_capsule}]
 spine_02 [{shapeType=Capsule, length=8.25, radius=8.75, shapeName=spine_02_capsule}]
 spine_03 [{shapeType=Capsule, length=7.75, radius=9.25, shapeName=spine_03_capsule}]
}
constraints [
 {bone1Name=dyn_chain_01, bone2Name=dyn_chain_mstr, swing1LimitDegrees=20, swing2LimitDegrees=45, twistLimitDegrees=5}
 {bone1Name=dyn_chain_02, bone2Name=dyn_chain_01, swing1LimitDegrees=10, swing2LimitDegrees=45, twistLimitDegrees=5}
 {bone1Name=dyn_chain_03, bone2Name=dyn_chain_02, swing1LimitDegrees=20, swing2LimitDegrees=45, twistLimitDegrees=5}
 {bone1Name=dyn_chain_04, bone2Name=dyn_chain_03, swing1LimitDegrees=20, swing2LimitDegrees=45, twistLimitDegrees=5}
 {bone1Name=dyn_chain_05, bone2Name=dyn_chain_04, swing1LimitDegrees=20, swing2LimitDegrees=45, twistLimitDegrees=5}
 {bone1Name=dyn_chain_06, bone2Name=dyn_chain_05, swing1LimitDegrees=45, swing2LimitDegrees=45, twistLimitDegrees=5}
]
