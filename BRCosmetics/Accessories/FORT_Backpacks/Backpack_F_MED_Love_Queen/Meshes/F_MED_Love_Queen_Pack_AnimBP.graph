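# Animation Blueprint for the Love Queen backpack cosmetic, adding secondary motion and fitting it to the wearer's body.
# Event graph copies pelvis and spine world transforms from the attach parent, flags the dual-wield harvesting montage, and computes state-based gravity.
# AnimGraph offsets the pack, with an interpolated nudge during that montage, then applies pelvis position and spine transforms.
# Eight spring-driven AnimDynamics fan bones feed a RigidBody node using that gravity as external force.
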AnimBlueprint /BRCosmetics/Accessories/FORT_Backpacks/Backpack_F_MED_Love_Queen/Meshes/F_MED_Love_Queen_Pack_AnimBP
target_skeleton=&/F_MED_Love_Queen_Pack_Skeleton
graphs [
 AnimGraph {nodes [
   n1 = "Transform (Modify) Bone - Bone: pack"(ComponentPose=Local To Component(LocalPose=Input Pose() {name=InPose_0}), Translation="0,-1.8458,0", Rotation="0,0,0") {
    translationMode=BMM_Replace
    translationSpace=BCS_ParentBoneSpace
    alphaBoolBlend {blendOption=Linear}
    boneToModify {boneName=pack}
    translation {y=-1.8458}
   }
   n2 = "Transform (Modify) Bone - Bone: dyn_pack_mstr"(
    ComponentPose=@n1
    Alpha=Get is_dualWield_montage() -> (is_dualWield_montage:double)
    Translation="0,-0.5,0"
    Rotation="10,-10,0"
    Scale="1,1,1"
   ) {
    rotationMode=BMM_Additive
    translationMode=BMM_Additive
    alphaBoolBlend {blendOption=Linear}
    alphaScaleBiasClamp {bInterpResult=true, interpSpeedDecreasing=8, interpSpeedIncreasing=8}
    boneToModify {boneName=dyn_pack_mstr}
    rotation {pitch=10, yaw=-10}
    translation {y=-0.5}
   }
   n3 = Blend Poses by int!(
    BlendPose_0:Pose Link Structure=Component To Local(ComponentPose=@n2)
    BlendPose_1:Pose Link Structure=Input Pose()
    BlendTime_0:float=0.1
    BlendTime_1:float=0.1
    ActiveChildIndex:int=0
   ) -> (Pose:Pose Link Structure) {
    is=title AnimGraphNode_BlendListByInt
    blendType=Linear
    blendPose [{linkId=-1, sourceLinkId=-1}, {linkId=-1, sourceLinkId=-1}]
    blendTime [0.1,0.1]
   }
   n4 = "Transform (Modify) Bone - Bone: pelvis"(
    ComponentPose=Local To Component(LocalPose=@n3)
    Alpha=1
    Translation=Get pevlis_transform() -> (pevlis_transform:Vector)
    Rotation=Get pelvis_rotator() -> (pelvis_rotator:Rotator)
    Scale="1,1,1"
   ) {translationMode=BMM_Replace, translationSpace=BCS_WorldSpace, alphaBoolBlend {blendOption=Linear}, boneToModify {boneName=pelvis}}
   n5 = "Transform (Modify) Bone - Bone: spine_01"(
    ComponentPose=@n4
    Alpha=1
    Translation=Get spine_01_transform() -> (spine_01_transform:Vector)
    Rotation=Get spine_01_rotator() -> (spine_01_rotator:Rotator)
    Scale="1,1,1"
   ) {
    rotationMode=BMM_Replace
    rotationSpace=BCS_WorldSpace
    translationMode=BMM_Replace
    translationSpace=BCS_WorldSpace
    alphaBoolBlend {blendOption=Linear}
    boneToModify {boneName=spine_01}
   }
   n6 = "Transform (Modify) Bone - Bone: spine_02"(
    ComponentPose=@n5
    Alpha=1
    Translation=Get spine_02_transform() -> (spine_02_transform:Vector)
    Rotation=Get spine_02_rotator() -> (spine_02_rotator:Rotator)
    Scale="1,1,1"
   ) {
    rotationMode=BMM_Replace
    rotationSpace=BCS_WorldSpace
    translationMode=BMM_Replace
    translationSpace=BCS_WorldSpace
    alphaBoolBlend {blendOption=Linear}
    boneToModify {boneName=spine_02}
   }
   n7 = "Transform (Modify) Bone - Bone: spine_03"(
    ComponentPose=@n6
    Alpha=1
    Translation=Get spine_03_transform() -> (spine_03_transform:Vector)
    Rotation=Get spine_03_rotator() -> (spine_03_rotator:Rotator)
    Scale="1,1,1"
   ) {
    rotationMode=BMM_Replace
    rotationSpace=BCS_WorldSpace
    translationMode=BMM_Replace
    translationSpace=BCS_WorldSpace
    alphaBoolBlend {blendOption=Linear}
    boneToModify {boneName=spine_03}
   }
   n8 = Anim Dynamics - Bone: fan_01(ComponentPose=@n7, Alpha=1, ExternalForce="0,0,0") {
    angularDampingOverride=0.99
    angularSpringConstant=10000
    bAngularSpring=true
    bGravityOverrideInSimSpace=true
    bOverrideAngularDamping=true
    bUseGravityOverride=true
    gravityScale=0
    simulationSpace=BoneRelative
    alphaBoolBlend {blendOption=Linear}
    boundBone {boneName=fan_01}
    componentAppliedLinearAccClamp=<1500 1500 1500>
    componentLinearAccScale=<0.1 0.1 0.1>
    componentLinearVelScale=<1 1 1>
    gravityOverride {z=-100}
    physicsBodyDefinitions [{
      sphereCollisionRadius=0
      boundBone {boneName=fan_01}
      boxExtents=<35 10 10>
      constraintSetup {angularLimitsMax=<0 15 0>, angularLimitsMin=<0 -5 0>, angularTarget=<1 0 0>}
      localJointOffset=<1.25 0 0>
     }]
    relativeSpaceBone {boneName=dyn_pack_mstr}
   }
   n9 = Anim Dynamics - Bone: fan_02(ComponentPose=@n8, Alpha=1, ExternalForce="0,0,0") {
    angularDampingOverride=0.99
    angularSpringConstant=10000
    bAngularSpring=true
    bGravityOverrideInSimSpace=true
    bOverrideAngularDamping=true
    bUseGravityOverride=true
    gravityScale=0
    simulationSpace=BoneRelative
    alphaBoolBlend {blendOption=Linear}
    boundBone {boneName=fan_02}
    componentAppliedLinearAccClamp=<1500 1500 1500>
    componentLinearAccScale=<0.1 0.1 0.1>
    componentLinearVelScale=<1 1 1>
    gravityOverride {z=-100}
    physicsBodyDefinitions [{
      sphereCollisionRadius=0
      boundBone {boneName=fan_02}
      boxExtents=<35 10 10>
      constraintSetup {angularLimitsMax=<0 20 0>, angularLimitsMin=<0 -5 0>, angularTarget=<1 0 0>}
      localJointOffset=<1.25 0 0>
     }]
    relativeSpaceBone {boneName=dyn_pack_mstr}
   }
   n10 = Anim Dynamics - Bone: fan_03(ComponentPose=@n9, Alpha=1, ExternalForce="0,0,0") {
    angularDampingOverride=0.99
    angularSpringConstant=10000
    bAngularSpring=true
    bGravityOverrideInSimSpace=true
    bOverrideAngularDamping=true
    bUseGravityOverride=true
    gravityScale=0
    simulationSpace=BoneRelative
    alphaBoolBlend {blendOption=Linear}
    boundBone {boneName=fan_03}
    componentAppliedLinearAccClamp=<1500 1500 1500>
    componentLinearAccScale=<0.1 0.1 0.1>
    componentLinearVelScale=<1 1 1>
    gravityOverride {z=-100}
    physicsBodyDefinitions [{
      sphereCollisionRadius=0
      boundBone {boneName=fan_03}
      boxExtents=<35 10 10>
      constraintSetup {angularLimitsMax=<0 20 0>, angularLimitsMin=<0 -5 0>, angularTarget=<1 0 0>}
      localJointOffset=<1.25 0 0>
     }]
    relativeSpaceBone {boneName=dyn_pack_mstr}
   }
   n11 = Anim Dynamics - Bone: fan_04(ComponentPose=@n10, Alpha=1, ExternalForce="0,0,0") {
    angularDampingOverride=0.99
    angularSpringConstant=10000
    bAngularSpring=true
    bGravityOverrideInSimSpace=true
    bOverrideAngularDamping=true
    bUseGravityOverride=true
    gravityScale=0
    simulationSpace=BoneRelative
    alphaBoolBlend {blendOption=Linear}
    boundBone {boneName=fan_04}
    componentAppliedLinearAccClamp=<1500 1500 1500>
    componentLinearAccScale=<0.1 0.1 0.1>
    componentLinearVelScale=<1 1 1>
    gravityOverride {z=-100}
    physicsBodyDefinitions [{
      sphereCollisionRadius=0
      boundBone {boneName=fan_04}
      boxExtents=<35 10 10>
      constraintSetup {angularLimitsMax=<0 20 0>, angularLimitsMin=<0 -5 0>, angularTarget=<1 0 0>}
      localJointOffset=<1.25 0 0>
     }]
    relativeSpaceBone {boneName=dyn_pack_mstr}
   }
   n12 = Anim Dynamics - Bone: fan_05(ComponentPose=@n11, Alpha=1, ExternalForce="0,0,0") {
    angularDampingOverride=0.99
    angularSpringConstant=10000
    bAngularSpring=true
    bGravityOverrideInSimSpace=true
    bOverrideAngularDamping=true
    bUseGravityOverride=true
    gravityScale=0
    simulationSpace=BoneRelative
    alphaBoolBlend {blendOption=Linear}
    boundBone {boneName=fan_05}
    componentAppliedLinearAccClamp=<1500 1500 1500>
    componentLinearAccScale=<0.1 0.1 0.1>
    componentLinearVelScale=<1 1 1>
    gravityOverride {z=-100}
    physicsBodyDefinitions [{
      sphereCollisionRadius=0
      boundBone {boneName=fan_05}
      boxExtents=<35 10 10>
      constraintSetup {angularLimitsMax=<0 20 0>, angularLimitsMin=<0 -5 0>, angularTarget=<1 0 0>}
      localJointOffset=<1.25 0 0>
     }]
    relativeSpaceBone {boneName=dyn_pack_mstr}
   }
   n13 = Anim Dynamics - Bone: fan_06(ComponentPose=@n12, Alpha=1, ExternalForce="0,0,0") {
    angularDampingOverride=0.99
    angularSpringConstant=10000
    bAngularSpring=true
    bGravityOverrideInSimSpace=true
    bOverrideAngularDamping=true
    bUseGravityOverride=true
    gravityScale=0
    simulationSpace=BoneRelative
    alphaBoolBlend {blendOption=Linear}
    boundBone {boneName=fan_06}
    componentAppliedLinearAccClamp=<1500 1500 1500>
    componentLinearAccScale=<0.1 0.1 0.1>
    componentLinearVelScale=<1 1 1>
    gravityOverride {z=-100}
    physicsBodyDefinitions [{
      sphereCollisionRadius=0
      boundBone {boneName=fan_06}
      boxExtents=<35 10 10>
      constraintSetup {angularLimitsMax=<0 20 0>, angularLimitsMin=<0 -5 0>, angularTarget=<1 0 0>}
      localJointOffset=<1.25 0 0>
     }]
    relativeSpaceBone {boneName=dyn_pack_mstr}
   }
   n14 = Anim Dynamics - Bone: fan_07(ComponentPose=@n13, Alpha=1, ExternalForce="0,0,0") {
    angularDampingOverride=0.99
    angularSpringConstant=10000
    bAngularSpring=true
    bGravityOverrideInSimSpace=true
    bOverrideAngularDamping=true
    bUseGravityOverride=true
    gravityScale=0
    simulationSpace=BoneRelative
    alphaBoolBlend {blendOption=Linear}
    boundBone {boneName=fan_07}
    componentAppliedLinearAccClamp=<1500 1500 1500>
    componentLinearAccScale=<0.1 0.1 0.1>
    componentLinearVelScale=<1 1 1>
    gravityOverride {z=-100}
    physicsBodyDefinitions [{
      sphereCollisionRadius=0
      boundBone {boneName=fan_07}
      boxExtents=<35 10 10>
      constraintSetup {angularLimitsMax=<0 20 0>, angularLimitsMin=<0 -5 0>, angularTarget=<1 0 0>}
      localJointOffset=<1.25 0 0>
     }]
    relativeSpaceBone {boneName=dyn_pack_mstr}
   }
   n15 = Anim Dynamics - Bone: fan_08(ComponentPose=@n14, Alpha=1, ExternalForce="0,0,0") {
    angularDampingOverride=0.99
    angularSpringConstant=10000
    bAngularSpring=true
    bGravityOverrideInSimSpace=true
    bOverrideAngularDamping=true
    bUseGravityOverride=true
    gravityScale=0
    simulationSpace=BoneRelative
    alphaBoolBlend {blendOption=Linear}
    boundBone {boneName=fan_08}
    componentAppliedLinearAccClamp=<1500 1500 1500>
    componentLinearAccScale=<0.1 0.1 0.1>
    componentLinearVelScale=<1 1 1>
    gravityOverride {z=-100}
    physicsBodyDefinitions [{
      sphereCollisionRadius=0
      boundBone {boneName=fan_08}
      boxExtents=<35 10 10>
      constraintSetup {angularLimitsMax=<0 20 0>, angularLimitsMin=<0 -5 0>, angularTarget=<1 0 0>}
      localJointOffset=<1.25 0 0>
     }]
    relativeSpaceBone {boneName=dyn_pack_mstr}
   }
   n16 = RigidBody(ComponentPose=@n15, ExternalForce=Get gravity_ovverride() -> (gravity_ovverride:Vector)) {
    bClampLinearTranslationLimitToRefPose=true
    bOverrideWorldGravity=true
    overridePhysicsAsset=&/F_MED_Love_Queen_Pack_Physics
    simulationSpace=BaseBoneSpace
    alphaBoolBlend {blendOption=Linear}
    baseBoneRef {boneName=pack}
    componentAppliedLinearAccClamp=<3000 3000 3000>
    componentLinearAccScale=<0.4 0.3 0.3>
    componentLinearVelScale=<0.4 0.3 0.3>
    overrideWorldGravity {z=-980}
   }
   Output Pose(Result=Component To Local(ComponentPose=@n16)) {layerGroup=None}
  ]}
 EventGraph {nodes [
   Event BlueprintUpdateAnimation()
   > n1 = Sequence()
   > n2 = Gravity Override State Function DEPRECATED(CharacterPartAnimInstance=Self-Reference(), GameTimeInSeconds=GetGameTimeInSeconds(), GravityOverrideParams=Get Gravity Override Params())
   > Set gravity_ovverride(gravity_ovverride:Vector=@n2) -> (Output_Get:Vector)
   n3 = GetOwnerSkeletalMeshComponent()
   n4 = GetDisplayName(Object=GetCurrentActiveMontage!(self:Anim Instance Object Reference=GetAnimInstance(self=@n3)) -> (ReturnValue:Anim Montage Object Reference))
   > Set is_dualWield_montage(is_dualWield_montage:double=SelectFloat(A=1, B=0, bPickA="Equal, Case Insensitive (String)"(A=@n4, B=DualWield_Harvesting_Combo_M))) -> (Output_Get:double)
   n5 = GetPartSkeletalMeshComponent()
   Is Valid(exec=@n1.then_1, InputObject=@n5)
   n6 = GetAttachParent!(self:Scene Component Object Reference=@n5) -> (ReturnValue:Scene Component Object Reference)
   > Is Valid(InputObject=@n6)
   n7 = BreakTransform(InTransform=GetSocketTransform(self=@n6, InSocketName=Pelvis, TransformSpace=RTS_World))
   > Set pevlis_transform(pevlis_transform:Vector=@n7) -> (Output_Get:Vector)
   > Set pelvis_rotator(pelvis_rotator:Rotator=@n7.Rotation) -> (Output_Get:Rotator)
   n8 = BreakTransform(InTransform=GetSocketTransform(self=@n6, InSocketName=Spine_01, TransformSpace=RTS_World))
   > Set spine_01_transform(spine_01_transform:Vector=@n8) -> (Output_Get:Vector)
   > Set spine_01_rotator(spine_01_rotator:Rotator=@n8.Rotation) -> (Output_Get:Rotator)
   n9 = BreakTransform(InTransform=GetSocketTransform(self=@n6, InSocketName=Spine_02, TransformSpace=RTS_World))
   > Set spine_02_transform(spine_02_transform:Vector=@n9) -> (Output_Get:Vector)
   > Set spine_02_rotator(spine_02_rotator:Rotator=@n9.Rotation) -> (Output_Get:Rotator)
   n10 = BreakTransform(InTransform=GetSocketTransform(self=@n6, InSocketName=Spine_03, TransformSpace=RTS_World))
   > Set spine_03_transform(spine_03_transform:Vector=@n10) -> (Output_Get:Vector)
   > Set spine_03_rotator(spine_03_rotator:Rotator=@n10.Rotation) -> (Output_Get:Rotator)
  ]}
]
variables [
 Gravity Override Params {type=GravityOverrideParamsStruct, default {onGround_Moving {gravityOverride=<0 0 0>}, skydive_FloatingDown {gravityOverride=<0 0 980>}}}
 gravity_ovverride {type=Vector}
 is_dualWield_montage {type=number}
 pelvis_rotator {type=Rotator}
 pevlis_transform {type=Vector}
 spine_01_rotator {type=Rotator}
 spine_01_transform {type=Vector}
 spine_02_rotator {type=Rotator}
 spine_02_transform {type=Vector}
 spine_03_rotator {type=Rotator}
 spine_03_transform {type=Vector}
]
